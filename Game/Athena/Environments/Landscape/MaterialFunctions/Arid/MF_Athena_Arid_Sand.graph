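# Material function that builds the arid sand surface for Athena landscape materials, projected on world XY at a fixed tile size.
# A packed input texture supplies tinted sand color from blue and blends in reddish pebbles by alpha, with a no-pebbles color also output.
# Sand textures supply height (also packed into Opacity), spec, normal and switchable roughness.
# Outputs include full material attributes and a simple variant with only base color and constant roughness.
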MaterialFunction /Game/Athena/Environments/Landscape/MaterialFunctions/Arid/MF_Athena_Arid_Sand
graph {nodes [
  Constant3Vector() {constant=<rgba 0 0 1 0>}
  n1 = Abs(None=Constant() {r=1500})
  n2 = ComponentMask(None=Divide(A=WorldPosition() {worldPositionShaderOffset=WPT_ExcludeAllShaderOffsets}.XYZ, B=Multiply(A=@n1) {constB=-1})) {g=true, r=true}
  n3 = TextureSampleParameter2D(UVs=@n2) {parameterName=Arid_Sand_01, samplerSource=SSM_Wrap_WorldGroupSettings, texture=/Game/Environments/Sets/Arid/Textures/T_desert_sand_base_C}
  FunctionOutput(None=LinearInterpolate(A=Constant(), B=Constant() {r=0.75}, Alpha=@n3.G)) {outputName=Spec}
  n4 = TextureSample(
   UVs=@n2
   Tex=FunctionInput(Preview=TextureObject() {samplerType=SAMPLERTYPE_LinearColor, texture=/Game/Athena/Environments/Landscape/Textures/Arid/T_Athena_Arid_Rock_01_M}) {bUsePreviewValueAsDefault=true, inputType=FunctionInput_Texture2D}
  ) {samplerSource=SSM_Wrap_WorldGroupSettings}
  n5 = Multiply(A=ComponentMask(None=@n4.B) {b=true}, B=Constant3Vector() {constant=<rgba 1 0.7049 0.3585 0>})
  n6 = FunctionOutput(None=LinearInterpolate(A=@n5, B=Constant3Vector() {constant=<rgba 0.65 0.1313 0.0578 1>}, Alpha=ComponentMask(None=@n4.A) {a=true})) {outputName=BaseColor}
  FunctionOutput(None=MakeMaterialAttributes!(
    BaseColor=@n6
    Metallic
    Specular
    Roughness=Constant() {r=0.95}
    Anisotropy
    EmissiveColor
    Opacity
    OpacityMask
    Normal
    Tangent
    WorldPositionOffset
    SubsurfaceColor
    ClearCoat
    ClearCoatRoughness
    AmbientOcclusion
    Refraction
    CustomizedUVs_0
    CustomizedUVs_1
    CustomizedUVs_2
    CustomizedUVs_3
    CustomizedUVs_4
    CustomizedUVs_5
    CustomizedUVs_6
    CustomizedUVs_7
    PixelDepthOffset
    ShadingModel
    Displacement
   )) {outputName=Simple_Material}
  FunctionOutput(None=@n5) {outputName=BaseColor_NoPebbles}
  n7 = TextureSample(UVs=@n2) {samplerSource=SSM_Wrap_WorldGroupSettings, samplerType=SAMPLERTYPE_Masks, texture=/Game/Environments/Sets/Arid/Textures/T_desert_sand_base_S}
  n8 = LinearInterpolate(A=Constant() {r=-0.5}, B=Constant() {r=1.2}, Alpha=@n7.B)
  n9 = Constant() {r=1}
  n10 = FunctionInput(Preview=StaticBool() {value=true}) {bUsePreviewValueAsDefault=true, inputName=UseRoughness, inputType=FunctionInput_StaticBool}
  n11 = TextureSampleParameter2D(UVs=@n2) {
   parameterName=Arid_Sand_01_Normal
   samplerSource=SSM_Wrap_WorldGroupSettings
   samplerType=SAMPLERTYPE_Normal
   texture=/Game/Environments/Sets/Arid/Textures/T_desert_sand_base_N
  }
  FunctionOutput(None=MakeMaterialAttributes(
    BaseColor=@n6
    Roughness=StaticSwitch(True=@n8, False=@n9, Value=@n10)
    Opacity=FunctionOutput(None=@n3.R) {outputName=Height}
    Normal=FunctionOutput(None=@n11.RGB) {outputName=Normal}
   )) {outputName=Mat}
 ]}
