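# Simple arid landscape material for desert roads, usable on spline meshes and Nanite.
# An arid sand layer with overridden specular optionally blends with tinted grass, using a height lerp driven by vertex-color green and grass height.
# An optional desert road function processes that result.
# A separate UseTopo switch scales sand base color with the colormap divisor and does not feed the grass or road chain.
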Material /Game/Athena/Environments/Landscape/Material/Arid/M_AthenaDesertRoad_Simple
graph {
 properties {
  bUseMaterialAttributes=true
  bUsedWithNanite=true
  bUsedWithSplineMeshes=true
  pixelDepthOffsetMode=PDOM_Legacy
  refractionMethod=RM_IndexOfRefraction
  usedShadingModels=Default Lit
 }
 nodes [
  n1 = Abs(None=CollectionParameter() {collection=/Game/Athena/Environments/Landscape/MPC/MPC_Landscape, parameterName=Arid_Sand_TexSize})
  ComponentMask(None=Divide(A=WorldPosition() {worldPositionShaderOffset=WPT_ExcludeAllShaderOffsets}.XYZ, B=Multiply(A=@n1) {constB=-1})) {g=true, r=true}
  n2 = MaterialFunctionCall!(In, UseRoughness=StaticBool()) {materialFunction=/Game/Athena/Environments/Landscape/MaterialFunctions/Arid/MF_Athena_Arid_Sand}
  n3 = SetMaterialAttributes!(MaterialAttributes=@n2.Simple_Material, Specular=Constant() {desc=Landscape Arid override spec value, r=0.1}) {attributeSetTypes [#1]}
  n4 = MaterialFunctionCall() {materialFunction=/Game/Athena/Environments/Landscape/MaterialFunctions/Standard/MF_Athena_Grass_01}
  n5 = ComponentMask(None=CollectionParameter() {collection=&MPC_Landscape, parameterName=Arid_Grass_Tint}) {b=true, g=true, r=true}
  n6 = VertexColor()
  n7 = StaticSwitchParameter(
   True=MaterialFunctionCall!(
    Base Material=@n3
    Top Material=MaterialFunctionCall!(In=@n4.Result, Color=@n5) {materialFunction=/Game/Athena/Environments/Landscape/MaterialFunctions/MF_Athena_Material_BaseMultiply}.Result
    Alpha=MaterialFunctionCall(Transition Phase=@n6.G, Height Texture=@n4.Height) {materialFunction=&HeightLerp}.Alpha
   ) {materialFunction=/Engine/Functions/MaterialLayerFunctions/MatLayerBlend_Standard}.Blended Material
   False=@n3
  ) {defaultValue=true, parameterName=UseGrass}
  StaticSwitchParameter(
   True=MaterialFunctionCall!(In=@n7) {materialFunction=/Game/Athena/Environments/Landscape/MaterialFunctions/Arid/MF_Athena_DesertRoad}.Result
   False=@n7
  ) {defaultValue=true, parameterName=UseDesertRoad}
  n8 = GetMaterialAttributes(None=@n2.Simple_Material) {attributeGetTypes [#2, #1]}
  n9 = Multiply(B=CollectionParameter() {collection=&MPC_Landscape, parameterName=ColormapDivisor})
  StaticSwitchParameter(
   True=SetMaterialAttributes!(MaterialAttributes=@n8.MaterialAttributes, Base Color=Multiply(A=@n8.Base Color, B=@n9)) {attributeSetTypes [#2]}
   False=@n2.Simple_Material
  ) {defaultValue=true, parameterName=UseTopo}
 ]
}
